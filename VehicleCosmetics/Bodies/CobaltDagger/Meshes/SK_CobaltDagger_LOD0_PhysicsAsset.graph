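# Physics asset for LOD0 of the CobaltDagger vehicle body skeletal mesh in VehicleCosmetics.
# It holds a single body on the frame bone, shaped as one large, rotated capsule offset from the bone origin.
# It has no constraints, so it provides one coarse collision volume with no articulated parts.
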PhysicsAsset /VehicleCosmetics/Bodies/CobaltDagger/Meshes/SK_CobaltDagger_LOD0_PhysicsAsset
bodies {frame [{shapeType=Capsule, length=226.3502, radius=105.8472, center=<-9.7846 -55.0552 -3.8779>, rotation=<rot 0 85.6662 -73.443>}]}
constraints []
